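# Default Lit master material in the Bridge MSPresets folder, built as the virtual texture version of the default surface material.
# Samples Albedo, Metalness, DR and Normal virtual texture parameters, taking roughness from DR's green channel.
# Feeds them into MF_ObjAdjustments, shared from the M_MS_Default_Material folder, which supplies the material attributes.
# All samples except DR use the shared wrap sampler and the Texture Maps parameter group.
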Material /Bridge/MSPresets/M_MS_Default_Material_VT/M_MS_Default_Material_VT
graph {
 properties {bUseMaterialAttributes=true, pixelDepthOffsetMode=PDOM_Legacy, usedShadingModels=Default Lit}
 nodes [
  n1 = TextureSampleParameter2D() {
   group=07 - Texture Maps
   parameterName=Albedo
   samplerSource=SSM_Wrap_WorldGroupSettings
   samplerType=SAMPLERTYPE_VirtualColor
   sortPriority=0
   texture=/Bridge/MSPresets/MSVTTextures/DefaultDiffuse
  }
  n2 = TextureSampleParameter2D() {
   group=07 - Texture Maps
   parameterName=Metalness
   samplerSource=SSM_Wrap_WorldGroupSettings
   samplerType=SAMPLERTYPE_VirtualLinearColor
   sortPriority=1
   texture=/Bridge/MSPresets/MSVTTextures/BlackPlaceholder
  }
  n3 = TextureSampleParameter2D() {desc=DR, parameterName=DR, samplerType=SAMPLERTYPE_VirtualLinearColor, texture=/Bridge/MSPresets/MSVTTextures/WhitePlaceholder}
  n4 = TextureSampleParameter2D() {
   group=07 - Texture Maps
   parameterName=Normal
   samplerSource=SSM_Wrap_WorldGroupSettings
   samplerType=SAMPLERTYPE_VirtualNormal
   sortPriority=3
   texture=/Bridge/MSPresets/MSVTTextures/Placeholder_Normal
  }
  MaterialFunctionCall!(Albedo=@n1.RGB, Metallic=@n2.R, Roughness=@n3.G, Tangent Normal=@n4.RGB) {materialFunction=/Bridge/MSPresets/M_MS_Default_Material/Functions/MF_ObjAdjustments}
 ]
}
